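# Legacy material function for a particle-based full-screen transition, pinning geometry in front of the camera through WPO.
# Alpha switches between a sphere mask on scene world position, centred from the particle and scaled by camera distance and depth, and a view-direction mask.
# Pattern noise uses normal-blended triplanar or screen-space sampling, with single-plane fallbacks for forward and low-end platforms.
# It also outputs world position and normals, rebuilt from depth differences in forward rendering.
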MaterialFunction /DealershipSharedLibrary/FX/Legacy/MF_DLR_FullScreenTransition
graph {
 properties {previewBlendMode=BLEND_Translucent, userExposedCaption=Full Screen Transition, libraryCategoriesText [Coronado]}
 nodes [
  n1 = Normalize(VectorInput=CameraVectorWS())
  n2 = SceneDepth()
  n3 = Constant3Vector() {constant=<rgba 0 0 1 0>}
  n4 = DotProduct(A=@n1, B=Normalize(VectorInput=Transform(None=@n3) {transformSourceType=TRANSFORMSOURCE_View}))
  Add(A=Multiply(A=@n1, B=Divide(A=@n2.R, B=@n4)), B=CameraPositionWS())
  FunctionInput() {
   bUsePreviewValueAsDefault=true
   inputName=-----Forward render-----
   inputType=FunctionInput_Bool
   sortPriority=50
   previewValue=<0.0001 0 0 1>
  }
  FunctionOutput(None=MaterialFunctionCall() {materialFunction=&WorldPositionBehindTranslucency}.Result) {outputName=UVs}
  n5 = MaterialFunctionCall() {materialFunction=&WorldPositionBehindTranslucency}
  n6 = Length(None=Subtract(A=CameraPositionWS(), B=ParticlePositionWS())) {is=class MaterialExpressionLength}
  n7 = SceneDepth()
  n8 = ScalarParameter() {defaultValue=2, parameterName=SceneBias}
  n9 = LinearInterpolate(B=Add(A=Multiply(A=@n6) {constB=4}, B=Multiply(A=@n7.R, B=@n8) {constB=0.005}), Alpha=NamedRerouteUsage()) {constA=50000, constB=50000}
  n10 = Add(A=ParticlePositionWS(), B=Multiply(A=NamedRerouteUsage(), B=@n9))
  n11 = Multiply(A=@n9, B=NamedRerouteUsage()) {constB=0.2}
  n12 = CameraPositionWS()
  n13 = ParticlePositionWS()
  n14 = Distance(A=Normalize(VectorInput=Subtract(A=@n12, B=@n13)), B=CameraVectorWS())
  n15 = Subtract(A=OneMinus(None=Multiply(A=@n14) {constB=0.2}), B=NamedRerouteUsage())
  n16 = OneMinus(None=NamedRerouteUsage())
  FunctionOutput(None=StaticSwitch(True=SphereMask(A=@n5.Result, B=@n10, Radius=@n11) {hardnessPercent=0}, False=Subtract(A=@n15, B=@n16), Value=NamedRerouteUsage())) {outputName=Alpha, sortPriority=1}
  n17 = ComponentMask(None=ScreenPosition().ViewportUV) {g=true, r=true}
  n18 = Constant()
  n19 = FunctionInput() {
   bUsePreviewValueAsDefault=true
   description=Forward render only
   inputName=Depth Normal Bias
   inputType=FunctionInput_Scalar
   sortPriority=52
   previewValue=<0.0001 0 0 1>
  }
  n20 = AppendVector(A=@n18, B=@n19)
  n21 = SceneDepth(UVs=@n17)
  n22 = Subtract(A=SceneDepth(UVs=Add(A=@n17, B=@n20)).R, B=@n21.R)
  n23 = FunctionInput() {
   bUsePreviewValueAsDefault=true
   description=Forward render only
   inputName=Depth Normal Height
   inputType=FunctionInput_Scalar
   sortPriority=51
   previewValue=<0.1 0 0 1>
  }
  n24 = AppendVector(A=@n19, B=@n18)
  n25 = Subtract(A=SceneDepth(UVs=Add(A=@n24, B=@n17)).R, B=@n21.R)
  n26 = CrossProduct(A=AppendVector(A=Constant2Vector() {r=1}, B=Multiply(A=@n22, B=@n23)), B=AppendVector(A=Constant2Vector() {g=1}, B=Multiply(A=@n25, B=@n23)))
  n27 = ComponentMask(None=SceneTexture() {sceneTextureId=PPI_WorldNormal}.Color) {b=true, g=true, r=true}
  FunctionOutput(None=ShadingPathSwitch(Default=@n27, Deferred=@n27, Forward=Normalize(VectorInput=@n26))) {outputName=WorldNormal, sortPriority=4}
  n28 = Subtract(A=CameraPositionWS(), B=ParticlePositionWS())
  n29 = StaticSwitch(
   True=Normalize(VectorInput=@n28)
   False=MaterialFunctionCall() {materialFunction=&CameraDirectionVector}.Result
   Value=FunctionInput(Preview=StaticBool()) {bUsePreviewValueAsDefault=true, inputName=Align to Particle, inputType=FunctionInput_StaticBool, sortPriority=1}
  )
  n30 = FunctionInput(Preview=Constant() {r=3}) {bUsePreviewValueAsDefault=true, inputName=Camera Offset, inputType=FunctionInput_Scalar}
  FunctionOutput(None=Add(A=@n28, B=Multiply(A=@n29, B=@n30))) {description=Attach to Camera + Camera vector offset, outputName=WPO, sortPriority=5}
  n31 = NamedRerouteUsage()
  n32 = Subtract(A=CameraPositionWS(), B=ParticlePositionWS())
  n33 = LinearInterpolate(B=Multiply(A=Length(None=@n32)) {constB=2}, Alpha=NamedRerouteUsage()) {constA=50000, constB=50000}
  n34 = StaticBool() {value=true}
  MaterialFunctionCall!(
   Location=Add(A=ParticlePositionWS(), B=Multiply(A=@n31, B=@n33))
   Offset
   Radius=Multiply(A=NamedRerouteUsage(), B=@n33)
   Calculate Camera Inside=@n34
   Depth Biased Alpha=@n34
   FadeDistance
  ) {materialFunction=/Engine/Functions/Engine_MaterialFunctions01/Gradient/SphereGradient-3D}
  n35 = FunctionInput() {bUsePreviewValueAsDefault=true, inputName=Pattern Tile, sortPriority=31, previewValue=<1 1 1 1>}
  n36 = Multiply(
   A=@n35
   B=Multiply(A=MaterialFunctionCall() {materialFunction=&WorldPositionBehindTranslucency}.Result) {constB=0.0001}
  )
  n37 = ComponentMask(None=@n36) {b=true, g=true}
  n38 = FunctionInput(Preview=TextureObject() {samplerType=SAMPLERTYPE_Masks, texture=/Game/Global/Textures/Default/Noise/T_Generic_Perlin}) {bUsePreviewValueAsDefault=true, inputName=Pattern Texture, inputType=FunctionInput_Texture2D, sortPriority=30, previewValue=<1 1 1 1>}
  n39 = TextureSample(UVs=ComponentMask(None=@n36) {g=true, r=true}, Tex=@n38)
  n40 = Abs(None=SceneTexture() {sceneTextureId=PPI_WorldNormal}.Color)
  n41 = LinearInterpolate(
   A=LinearInterpolate(A=TextureSample(UVs=@n37, Tex=@n38).RGB, B=@n39.RGB, Alpha=ComponentMask(None=@n40) {b=true})
   B=TextureSample(UVs=ComponentMask(None=@n36) {b=true, r=true}, Tex=@n38).RGB
   Alpha=ComponentMask(None=@n40) {g=true}
  )
  n42 = ShadingPathSwitch(Default=@n41, Deferred=@n41, Forward=@n39.RGB)
  n43 = MaterialFunctionCall(Float2=MaterialFunctionCall() {materialFunction=&ScreenResolution}.Visible Resolution) {materialFunction=&BreakOutFloat2Components}
  n44 = Divide(A=@n43.R, B=@n43.R)
  n45 = MaterialFunctionCall(Float2=ScreenPosition().ViewportUV) {materialFunction=&BreakOutFloat2Components}
  n46 = Multiply(A=AppendVector(A=Multiply(A=@n44, B=@n45.R), B=@n45.G), B=ComponentMask(None=@n35) {g=true, r=true})
  n47 = FunctionInput(Preview=StaticBool() {value=true}) {bUsePreviewValueAsDefault=true, inputName=Use Word Space Mask, inputType=FunctionInput_StaticBool, previewValue=<1 0 0 1>}
  n48 = ChannelMaskParameter(None=StaticSwitch(
    True=MaterialFunctionCall("Xb1,PS4"=@n42, "PC-High&&Epic,XSX,PS5"=@n42, "PC-Perf,Android"=@n39.RGB, Switch=@n42, "PC-Low(NotPerf)&&Medium"=@n42) {materialFunction=&PlatformSwitch_1D}.Result
    False=TextureSample(UVs=@n46, Tex=@n38).RGB
    Value=NamedRerouteDeclaration(None=@n47) {name=Is Word Space, nodeColor=<rgba 0 1 0.0391 1>}
   ) {defaultValue=true}) {maskChannel=Blue, parameterName=NoiseChannel, defaultValue=<rgba 0 0 1 0>}
  n49 = FunctionInput() {bUsePreviewValueAsDefault=true, inputName=Pattern Amount, inputType=FunctionInput_Scalar, sortPriority=32, previewValue=<0.1 0 0 1>}
  NamedRerouteDeclaration(None=Multiply(A=Add(A=@n48.R) {constB=-0.5}, B=@n49)) {name=Noise, nodeColor=<rgba 1 0 0.0234 1>}
  n50 = ScalarParameter() {defaultValue=0.2, parameterName=DebugTimeSpeed}
  n51 = FunctionInput() {bUsePreviewValueAsDefault=true, inputName=Transition, inputType=FunctionInput_Scalar, previewValue=<0.5 1 1 1>}
  NamedRerouteDeclaration(None=StaticSwitchParameter(
    True=MaterialFunctionCall(Speed=@n50) {materialFunction=&DebugTimeSine}.Result
    False=@n51
   ) {parameterName=DebugTransition}) {name=Transition, nodeColor=<rgba 1 0.1036 0 1>}
  n52 = Divide(A=MaterialFunctionCall() {materialFunction=&WorldPositionBehindTranslucency}.Result) {constB=1024}
  n53 = Round(None=Abs(None=SceneTexture() {sceneTextureId=PPI_WorldNormal}.Color))
  n54 = Multiply(A=ComponentMask(None=@n52) {g=true, r=true}, B=ComponentMask(None=@n53) {b=true})
  n55 = Add(
   A=Multiply(A=ComponentMask(None=@n52) {b=true, g=true}, B=ComponentMask(None=@n53) {r=true}) {constB=0}
   B=Multiply(A=ComponentMask(None=@n52) {b=true, r=true}, B=ComponentMask(None=@n53) {g=true}) {constB=0}
  )
  NamedRerouteDeclaration(None=Divide(A=Add(A=@n54, B=@n55)) {constB=200}) {name=TriplanarUV, nodeColor=<rgba 1 0 0.2109 1>}
  n56 = MaterialFunctionCall(Float2=MaterialFunctionCall() {materialFunction=&ScreenResolution}.Visible Resolution) {materialFunction=&BreakOutFloat2Components}
  n57 = Divide(A=@n56.R, B=@n56.R)
  n58 = MaterialFunctionCall(Float2=MaterialFunctionCall(World Position=ParticlePositionWS()) {materialFunction=&TransformToClipSpace}.Buffer UVs for Scene Textures) {materialFunction=&BreakOutFloat2Components}
  n59 = Multiply(A=@n57, B=@n58.R)
  n60 = MaterialFunctionCall(Float2=ScreenPosition().ViewportUV) {materialFunction=&BreakOutFloat2Components}
  n61 = Multiply(A=@n57, B=@n60.R)
  SphereMask(
   A=Subtract(A=AppendVector(A=@n59, B=@n58.G), B=AppendVector(A=@n61, B=@n60.G))
   B=NamedRerouteUsage()
   Radius=Multiply(A=NamedRerouteUsage())
  ) {attenuationRadius=1, hardnessPercent=0}
 ]
}
