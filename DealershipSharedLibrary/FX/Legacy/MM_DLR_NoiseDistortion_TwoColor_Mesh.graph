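# Translucent unlit master material for Niagara mesh particles, producing a two-color, noise-distorted dissolve effect.
# Panning noise offsets base texture UVs, then a smoothstep threshold shapes the mask that blends dark and light colors.
# Fresnel, particle color and depth fade modulate emissive or opacity, with switches for a Fresnel-only look and an Is Painted desaturation.
# Dynamic parameters drive dissolve, distortion and noise-based vertex displacement along normals, which quality switches limit to High and Epic.
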Material /DealershipSharedLibrary/FX/Legacy/MM_DLR_NoiseDistortion_TwoColor_Mesh
graph {
 properties {
  bAutomaticallySetUsageInEditor=false
  bGenerateSphericalParticleNormals=true
  bUsedWithNiagaraMeshParticles=true
  blendMode=BLEND_Translucent
  refractionMethod=RM_IndexOfRefraction
  shadingModel=MSM_Unlit
  usedShadingModels=Unlit
 }
 nodes [
  n1 = VectorParameter() {
   parameterName=Noise Tiling/Panner Control
   channelNames {a=Pan Speed Y, b=Pan Speed X, g=Tiling Y, r=Tiling X}
   defaultValue=<rgba 0.56 0.68 1 1>
  }
  n2 = Multiply(A=TextureCoordinate(), B=AppendVector(A=@n1.Tiling X, B=@n1.Tiling X))
  n3 = AppendVector(A=@n1.Pan Speed X, B=@n1.Pan Speed X)
  Add(A=@n2, B=@n3)
  n4 = VectorParameter() {
   group=4. WPO
   parameterName=WPO Noise UV Controls
   sortPriority=5
   channelNames {a=Panner V, b=Panner U, g=Tiling V, r=Tiling U}
   defaultValue=<rgba 1 1 0 0>
  }
  n5 = Multiply(A=TextureCoordinate(), B=AppendVector(A=@n4.Tiling U, B=@n4.Tiling U)) {constB=2}
  Add(A=@n5, B=ParticleRandom())
  Constant() {r=1}
  Desaturation()
  n6 = ScalarParameter() {defaultValue=0.79, group=1. Fresnel, parameterName=Fresnel - ExponentIn, sortPriority=0}
  n7 = ScalarParameter() {defaultValue=0.106, group=1. Fresnel, parameterName=BaseReflectFractionIn, sortPriority=1}
  n8 = ScalarParameter() {defaultValue=3, group=1. Fresnel, parameterName=Fresnel Power, sortPriority=2}
  n9 = DynamicParameter() {parameterIndex=3, defaultValue=<rgba 0 3 1 1>, paramNames [Is Painted, Fresnel Power, Param3, Param4]}
  n10 = ScalarParameter() {parameterName="Use Dynamic Param Fresnel Power (0-1)"}
  n11 = Saturate(None=Power(Base=Fresnel(ExponentIn=@n6, BaseReflectFractionIn=@n7), Exp=LinearInterpolate(A=@n8, B=@n9.Fresnel Power, Alpha=@n10) {constAlpha=0}))
  n12 = Multiply(
   A=Multiply(A=@n11, B=VectorParameter() {parameterName=FresnelColor, defaultValue=<rgba 2.3186 0.312 0.6851 1>}.RGB)
   B=ScalarParameter() {defaultValue=1, parameterName=Fresnel Outer Multiply}
  )
  n13 = Fresnel(
   ExponentIn=ScalarParameter() {defaultValue=0.79, group=1. Fresnel, parameterName=Fresnel - ExponentIn, sortPriority=0}
   BaseReflectFractionIn=ScalarParameter() {defaultValue=0.106, group=1. Fresnel, parameterName=BaseReflectFractionIn, sortPriority=1}
  )
  n14 = Multiply(
   A=Power(Base=OneMinus(None=@n13), Exp=ScalarParameter() {defaultValue=3, group=1. Fresnel, parameterName=Fresnel Inverted Power, sortPriority=2})
   B=Clamp(None=ScalarParameter() {defaultValue=1, parameterName="InnerFresnelAmount(0-1)"})
  )
  n15 = Add(A=@n12, B=@n14)
  n16 = DynamicParameter() {defaultValue=<rgba -0.4527 1 0 1>, paramNames [Dissolve,Exp,Distortion,Refraction]}
  n17 = ChannelMaskParameter(None=TextureSampleParameter2D(UVs=Panner(Coordinate=@n2, Speed=@n3)) {parameterName=Noise Texture, texture=/Game/Effects/Fort_Effects/Emotes/UltraLight/Texture/T_SoftNoise_BC7}.RGB) {maskChannel=Green, parameterName=Noise Texture Channel, sortPriority=11, defaultValue=<rgba 0 1 0 0>}
  n18 = Subtract(A=ComponentMask(None=@n17.R) {g=true, r=true}) {constB=0.5}
  n19 = DynamicParameter() {defaultValue=<rgba -0.4527 1 0 1>, paramNames [Dissolve,Exp,Distortion,Refraction]}
  n20 = DynamicParameter() {parameterIndex=1, paramNames [BaseScale_X,BaseScale_Y,BaseOffset_X,BaseOffset_Y]}
  n21 = Multiply(A=TextureCoordinate(), B=AppendVector(A=@n20.BaseScale_X, B=@n20.BaseScale_X))
  n22 = AppendVector(A=@n20.BaseOffset_X, B=@n20.BaseOffset_X)
  n23 = TextureSampleParameter2D(UVs=Add(A=Multiply(A=@n18, B=@n19.Distortion), B=Add(A=@n21, B=@n22))) {parameterName=Base Texture, texture=/Game/Effects/Fort_Effects/Textures/Tile/T_Inky_Smoke_Tile2}
  n24 = Power(
   Base=StaticSwitchParameter(True=ChannelMaskParameter(None=@n23.RGBA) {group=Base Texture, parameterName=Base Texture Channel, sortPriority=11}.R, False=@n23.RGB) {defaultValue=true, parameterName=Use Texture Channel, sortPriority=10}
   Exp=@n16.Exp
  )
  n25 = Constant3Vector() {constant=<rgba 1 1 1 0>}
  n26 = Multiply(
   A=MaterialFunctionCall!(
    Alpha=@n24
    Cutoff=DynamicParameter() {defaultValue=<rgba -0.4527 1 0 1>, paramNames [Dissolve,Exp,Distortion,Refraction]}.Dissolve
    Soft Transition=StaticBool() {value=true}
    Transition Range=ScalarParameter() {defaultValue=0.7269, parameterName=DissolveSmoothness}
    Normalized Alpha Range=StaticBool()
   ) {materialFunction=/DealershipSharedLibrary/FX/Materials/MaterialFunctions/MF_AlphaThreshold_Smoothstep}.OpacityMask
   B=@n25
  )
  n27 = Multiply(
   A=StaticSwitch(True=Constant() {r=1}, False=Saturate(None=@n15), Value=StaticBoolParameter() {parameterName=DisableFresnelFully})
   B=Multiply(A=ScalarParameter() {defaultValue=5, parameterName=Fresnel - Texture Influience multiplier}, B=@n26)
  )
  n28 = Multiply(
   A=@n27
   B=LinearInterpolate(
    A=VectorParameter() {parameterName=BaseColorDark, defaultValue=<rgba 0.4062 0.0381 0.1602 1>}.RGB
    B=VectorParameter() {parameterName=BaseColorLight, defaultValue=<rgba 0.3872 0.7473 1.6236 1>}.RGB
    Alpha=@n26
   )
  ) {constA=-0.024}
  n29 = DynamicParameter() {parameterIndex=3, defaultValue=<rgba 0 3 1 1>, paramNames [Is Painted, Fresnel Power, Param3, Param4]}
  n30 = StaticSwitch(
   True=Power(Base=@n15, Exp=@n16.Exp) {constExponent=1}
   False=LinearInterpolate(A=@n28, B=Desaturation(None=@n28), Alpha=Clamp(None=@n29.Is Painted))
   Value=StaticBoolParameter() {parameterName=Basic-Fresnel and WPO only}
  )
  n31 = ParticleColor()
  MaterialFunctionCall(Emissive=Desaturation(None=Multiply(A=@n30, B=@n31.RGB), Fraction=ScalarParameter() {parameterName=desatur})) {materialFunction=&MF_FX_EmissiveAdjustments}
  n32 = Saturate(None=@n27)
  n33 = StaticSwitchParameter(True=@n32, False=If(A=ScalarParameter() {parameterName=DissolveOpacitySwitch}, "A > B"=@n32, "A == B"=@n25, "A < B"=@n25)) {parameterName=DissolveOpacity}
  n34 = StaticBoolParameter() {parameterName=Basic-Fresnel and WPO only}
  n35 = StaticSwitch(True=@n25, False=@n33, Value=@n34)
  n36 = NamedRerouteUsage()
  n37 = StaticSwitch(True=@n36, False=Multiply(A=@n33, B=@n36), Value=@n34)
  n38 = ScalarParameter() {parameterName=Use Fresnel for Opacity, sliderMax=1}
  n39 = Multiply(
   A=@n31.A
   B=Multiply(A=LinearInterpolate(A=@n35, B=@n37, Alpha=@n38), B=ScalarParameter() {defaultValue=1, parameterName=Overall Opacity Mult})
  )
  n40 = StaticSwitchParameter(True=@n24, False=Desaturation(None=@n24)) {defaultValue=true, parameterName=Use Texture Channel, sortPriority=10}
  n41 = LinearInterpolate(
   A=ScalarParameter() {defaultValue=400, parameterName=DepthFade - Max}
   B=ScalarParameter() {defaultValue=50, parameterName=DepthFade - Min}
   Alpha=@n40
  ) {constA=400, constB=50}
  Multiply(A=@n39, B=QualitySwitch(Default=DepthFade(FadeDistance=@n41) {fadeDistanceDefault=400}, Low=Constant() {r=1}))
  n42 = ScalarParameter() {defaultValue=1, group=Opacity, parameterName=Angle Fade Min, sortPriority=12}
  n43 = ScalarParameter() {group=Opacity, parameterName=Angle Fade Max, sortPriority=13}
  n44 = Fresnel(ExponentIn=ScalarParameter() {defaultValue=0.5, group=Opacity, parameterName=Angle Fade Exp, sortPriority=11}) {baseReflectFraction=0, exponent=1}
  NamedRerouteDeclaration(None=StaticSwitchParameter(True=LinearInterpolate(A=@n42, B=@n43, Alpha=@n44), False=Constant() {r=1}) {defaultValue=true, group=Opacity, parameterName=Use Angle Fade, sortPriority=10}) {name=AngleFade, nodeColor=<rgba 0.013 0.0953 0.0931 1>}
  n45 = Saturate(None=@n14)
  n46 = ScalarParameter() {parameterName=Invert Alpha Fresnel}
  n47 = Multiply(
   A=Power(Base=LinearInterpolate(A=@n11, B=@n45, Alpha=@n46), Exp=ScalarParameter() {defaultValue=1, parameterName=Opacity Fresnel Pow})
   B=ScalarParameter() {defaultValue=1, parameterName=Opacity Fresnel Mult}
  )
  NamedRerouteDeclaration(None=StaticSwitch(True=@n12, False=Clamp(None=@n47), Value=StaticBoolParameter() {parameterName=DoubleFresnelOpacity})) {name=AlphaFresnel, nodeColor=<rgba 0.8047 1 0 1>}
  n48 = Constant() {r=1}
  n49 = DynamicParameter() {parameterIndex=2, defaultValue=<rgba 0 1 -200 400>, paramNames ["Use WPO (0-1)", WPO Multiplier, Min WPO, Max WPO]}
  n50 = TextureSampleParameter2D(UVs=Panner(Coordinate=@n5, Speed=AppendVector(A=@n4.Panner U, B=@n4.Panner U)) {speedX=0.2, speedY=0.15}) {
   group=4. WPO
   parameterName=WPO Noise
   samplerType=SAMPLERTYPE_LinearColor
   sortPriority=4
   texture=/Game/Effects/SnowSandInteraction/Textures/SSI_Perlin
  }
  n51 = Multiply(
   A=VertexNormalWS()
   B=Multiply(A=LinearInterpolate(A=@n49.Min WPO, B=@n49.Min WPO, Alpha=@n50.R) {constA=-0.3, constB=0.3}, B=@n49.WPO Multiplier)
  )
  QualitySwitch(Default=@n48, Low=@n48, High=@n51, Medium=@n48, Epic=@n51)
  n52 = Constant()
  n53 = LinearInterpolate(B=Multiply(A=@n51, B=Saturate(None=@n49."Use WPO (0-1)"))) {constAlpha=1}
  n54 = ScalarParameter() {group=3. Modifiers, parameterName=CameraOffset}
  QualitySwitch(
   Default=@n52
   Low=@n52
   High=@n53
   Medium=@n52
   Epic=Add(
    A=MaterialFunctionCall(Offset Amount=@n54) {materialFunction=/Engine/Functions/Engine_MaterialFunctions02/WorldPositionOffset/CameraOffset}.Clamped Camera Offset
    B=@n53
   )
  )
  Saturate(None=@n39)
  Saturate(None=@n40)
  ScalarParameter() {defaultValue=500, group=4. WPO, parameterName=WPO Noise Max, sortPriority=7}
  ScalarParameter() {parameterName=frbugDissolve}
  ScalarParameter() {defaultValue=-100, group=4. WPO, parameterName=WPO Noise Min, sortPriority=6}
 ]
}
