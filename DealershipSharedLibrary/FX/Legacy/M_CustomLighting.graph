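# Legacy unlit, masked Niagara sprite material that fakes lit, metallic-capable particles with hand-built lighting.
# A SubUV normal map drives directional shading against a fixed Z vector, plus half-vector specular and optional VC_MF_WorldEnv reflections.
# MetallicInfluence blends particle-tinted BaseColor toward a Fresnel-style particle-color rim, and lets particle alpha fade lighting down to specular only.
# The color and lighting sum feeds MF_FX_EmissiveAdjustments.
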Material /DealershipSharedLibrary/FX/Legacy/M_CustomLighting
graph {
 properties {bAutomaticallySetUsageInEditor=false, bUsedWithNiagaraSprites=true, blendMode=BLEND_Masked, shadingModel=MSM_Unlit, usedShadingModels=Unlit}
 nodes [
  n1 = TextureSampleParameterSubUV!(UVs, Apply View MipBias) {group=Textures, parameterName=NormalMap, samplerType=SAMPLERTYPE_Normal, texture=/Game/Global/Textures/Default/Tileable_Patterns/1Normal}
  n2 = Transform(None=@n1.RGB)
  n3 = Constant3Vector() {constant=<rgba 0 0 1 0>}
  n4 = DotProduct(A=@n2, B=@n3)
  ConstantBiasScale(None=@n4) {scale=1}
  n5 = TextureSampleParameter2D() {group=Textures, parameterName=BaseColor}
  n6 = ParticleColor()
  n7 = ScalarParameter() {
   desc="Controls when a texture's color is multiplied by the particle color. Only works if 'MetallicInfluence' = 0."
   group=Controls
   parameterName=ColorInfluence
   sliderMax=1
  }
  n8 = LinearInterpolate(A=@n5.RGB, B=Multiply(A=@n5.RGB, B=@n6.RGB), Alpha=@n7)
  n9 = DotProduct(A=CameraVectorWS(), B=NamedRerouteUsage())
  n10 = Multiply(A=Saturate(None=OneMinus(None=@n9)), B=Power(Base=@n6.RGB) {constExponent=1.5})
  n11 = ScalarParameter() {
   defaultValue=1
   desc=Controls whether the material represents a metallic element or not.
   group=Controls
   parameterName=MetallicInfluence
   sliderMax=1
  }
  n12 = Multiply(
   A=Add(A=@n4, B=ScalarParameter() {defaultValue=1, group=DirLight, parameterName=Bias})
   B=ScalarParameter() {defaultValue=1, group=DirLight, parameterName=Scale}
  )
  n13 = Constant()
  n14 = Multiply(
   A=MaterialFunctionCall!(Desaturate=Constant() {r=1}, "Tint (Curve)"=@n13, Tint Amount=@n13, Cubemap) {materialFunction=/DealershipSharedLibrary/FX/Materials/MaterialFunctions/VC_MF_WorldEnv}.LocalSpace
   B=Constant() {r=1}
  )
  n15 = ScalarParameter() {
   desc=Controls whether the material is affected by reflections from the 'WorldEnv' MF. Typically used for metallic objects.
   group=Controls
   parameterName=ReflectionsInfluence
   sliderMax=1
  }
  n16 = NamedRerouteUsage()
  n17 = Add(A=LinearInterpolate(A=@n12, B=Multiply(A=@n14, B=@n12), Alpha=@n15), B=@n16)
  n18 = LinearInterpolate(A=@n17, B=@n16, Alpha=ParticleColor().A)
  n19 = ScalarParameter() {
   defaultValue=1
   desc=Controls whether the material represents a metallic element or not.
   group=Controls
   parameterName=MetallicInfluence
   sliderMax=1
  }
  MaterialFunctionCall(Emissive=Add(A=LinearInterpolate(A=@n8, B=@n10, Alpha=@n11), B=LinearInterpolate(A=@n17, B=@n18, Alpha=@n19)) {constB=0}) {materialFunction=&MF_FX_EmissiveAdjustments}
  NamedRerouteDeclaration(None=@n3) {name=DirLightNormal, nodeColor=<rgba 0.7344 0 1 1>}
  NamedRerouteDeclaration(None=@n2) {name=DirLightVector, nodeColor=<rgba 0.7344 0 1 1>}
  n20 = Transform(None=NamedRerouteUsage())
  n21 = CameraVectorWS()
  n22 = DotProduct(A=NamedRerouteUsage(), B=Normalize(VectorInput=Add(A=@n20, B=@n21)))
  n23 = ScalarParameter() {defaultValue=24, group=CustomSpecular, parameterName=SpecularPower}
  NamedRerouteDeclaration(None=Multiply(A=ScalarParameter() {defaultValue=10, group=CustomSpecular, parameterName=SpecularIntensity}, B=Power(Base=@n22, Exp=@n23)) {constA=2}) {name=Specular, nodeColor=<rgba 0.7344 0 1 1>}
  NamedRerouteDeclaration(None=@n1.RGB) {name=DirLightNormaTex, nodeColor=<rgba 0.7344 0 1 1>}
  NamedRerouteDeclaration(None=@n1.RGB) {name=Normal, nodeColor=<rgba 0.7344 0 1 1>}
  NamedRerouteUsage()
  ScalarParameter() {
   defaultValue=1
   desc=Default Metallic Shader Value. Used when Shading Model is set to 'Default Lit'.
   group=Default
   parameterName=Metallic
   sliderMax=1
  }
  ScalarParameter() {
   defaultValue=0.03
   desc=Default Roughness Shader Value. Used when Shading Model is set to 'Default Lit'.
   group=Default
   parameterName=Roughness
   sliderMax=1
  }
  TextureSampleParameter2D() {group=Textures, parameterName=Opacity}
 ]
}
