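# Translucent Niagara material for drift trail mist, built on a smoke flipbook texture and usable on sprites, ribbons and mesh particles.
# Opacity blends texture alpha or a selected RGB channel, scales it by particle alpha and dynamic Mask, optionally subtracts Dissolve, then feeds a depth-fade function.
# Color comes from the multiplied, desaturated texture, tinted by particle color and dynamic Brightness, then passed through an emissive-adjustment function.
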Material /DealershipSharedLibrary/FX/Legacy/M_DriftTrail_Mist
graph {
 properties {
  bAutomaticallySetUsageInEditor=false
  bUsedWithNiagaraMeshParticles=true
  bUsedWithNiagaraRibbons=true
  bUsedWithNiagaraSprites=true
  blendMode=BLEND_Translucent
  usedShadingModels=Default Lit
 }
 nodes [
  MaterialExpressionVertexColor_0 = ParticleColor()
  n1 = TextureSampleParameter2D() {group=Main Texture, parameterName=Main Texture, texture=/DealershipSharedLibrary/FX/Textures/Flipbook/T_VCVFX_SmokeHQ_N}
  n2 = LinearInterpolate(
   A=@n1.A
   B=DotProduct(A=@n1.RGB, B=VectorParameter() {group=Alpha, parameterName=RGB Channel Alpha Selector, defaultValue=<rgba 1 0 0 1>}.RGB)
   Alpha=ScalarParameter() {group=Alpha, parameterName="0=UseAlphaForOpacity / 1=UseRGB ForOpacity"}
  )
  n3 = DynamicParameter() {paramNames [Mask,Param2,Dissolve,Brightness]}
  MaterialExpressionConstantClamp_0 = Saturate(None=Multiply(A=@MaterialExpressionVertexColor_0.A, B=Multiply(A=@n2, B=@n3.Mask)))
  MaterialFunctionCall(
   Opacity In=LinearInterpolate(
    A=@MaterialExpressionConstantClamp_0
    B=Subtract(A=@MaterialExpressionConstantClamp_0, B=@n3.Dissolve)
    Alpha=ScalarParameter() {parameterName="0=NoDissolve/1=Dissolve"}
   )
   DepthFade Distance=ScalarParameter() {defaultValue=100, group=Depth Fade, parameterName=DepthFade Distance}
  ) {materialFunction=&MF_VCVFX_DepthFade}
  n4 = Desaturation(None=Multiply(A=@n1.RGB, B=ScalarParameter() {defaultValue=2, group=Main Texture, parameterName=Texture Multiplier}) {constB=2}) {luminanceFactors=<rgba 0.3 0.59 0.11 0>}
  MaterialFunctionCall(Emissive=Multiply(A=Multiply(A=@n4, B=@MaterialExpressionVertexColor_0.RGB), B=@n3.Brightness)) {materialFunction=&MF_FX_EmissiveAdjustments}
 ]
}
