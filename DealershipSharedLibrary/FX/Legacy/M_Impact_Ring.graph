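# Unlit translucent material for Niagara sprite and mesh particles that draws an animated, noisy impact ring.
# Packed noise pans in radial coordinates and, scaled by a hollow-center distance mask, drives the soft-noise lookup for emissive, tinted by particle color.
# That mask and a Sphere_Power edge falloff shape opacity, softened by depth fade.
# Dynamic parameters control brightness, opacity and speed, and a shared emissive-adjustment function processes the output.
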Material /DealershipSharedLibrary/FX/Legacy/M_Impact_Ring
graph {
 properties {
  bAutomaticallySetUsageInEditor=false
  bUseMaterialAttributes=true
  bUsedWithNiagaraMeshParticles=true
  bUsedWithNiagaraSprites=true
  blendMode=BLEND_Translucent
  shadingModel=MSM_Unlit
  usedShadingModels=Unlit
 }
 nodes [
  n1 = DynamicParameter() {defaultValue=<rgba 1 4 1 1>, paramNames [Brightness,Opacity,Speed,Param4]}
  n2 = AppendVector(A=Constant() {r=2}, B=Constant() {r=0.25})
  n3 = MaterialFunctionCall(Vector or UVs=Add(A=TextureCoordinate(), B=Constant() {r=-0.5})) {materialFunction=&VectorToRadialValue}
  n4 = Add(A=@n3.Radial Coordinates, B=@n3.Radial Coordinates)
  n5 = Time()
  n6 = TextureSample(UVs=Panner(Coordinate=Multiply(A=@n2, B=@n4), Time=Multiply(A=@n1.Speed, B=@n5), Speed=VectorParameter() {defaultValue=<rgba -0.1 0.4 0 1>}.RGB)) {samplerType=SAMPLERTYPE_Masks, texture=/Game/Effects/Fort_Effects/Textures/T_PackedNoise}
  n7 = Add(A=TextureCoordinate(), B=Constant() {r=-0.5})
  n8 = SquareRoot(None=DotProduct(A=@n7, B=@n7))
  n9 = Clamp(None=ConstantBiasScale(None=Clamp(None=@n8)) {bias=-0.15, scale=6})
  n10 = TextureSample(UVs=Multiply(A=AppendVector(A=@n6.R, B=@n6.R), B=@n9)) {texture=/Game/Effects/Fort_Effects/Emotes/UltraLight/Texture/T_SoftNoise_BC7}
  n11 = Add(A=Multiply(A=@n10.R, B=@n10.R), B=Multiply(A=@n10.G, B=@n10.G))
  n12 = Multiply(A=@n1.Brightness, B=Clamp(None=ConstantBiasScale(None=@n11) {bias=0.05, scale=3}))
  n13 = ParticleColor()
  n14 = Clamp(None=Multiply(A=@n9, B=@n9))
  n15 = Add(A=TextureCoordinate() {uTiling=2, vTiling=2}, B=Constant() {r=-1})
  n16 = DotProduct(A=@n15, B=@n15)
  n17 = Power(Base=OneMinus(None=SquareRoot(None=@n16)), Exp=ScalarParameter() {defaultValue=2, parameterName=Sphere_Power})
  n18 = Multiply(
   A=@n13.A
   B=Multiply(A=Multiply(A=@n14, B=@n17), B=DynamicParameter() {defaultValue=<rgba 1 4 1 1>, paramNames [Brightness,Opacity,Speed,Param4]}.Opacity)
  )
  n19 = ScalarParameter() {defaultValue=250, parameterName=FadeDistance}
  MaterialFunctionCall!(In=MakeMaterialAttributes!(
    BaseColor
    Metallic
    Specular
    Roughness
    Anisotropy
    EmissiveColor=Multiply(A=@n12, B=@n13.RGB)
    Opacity=DepthFade(Opacity=@n18, FadeDistance=@n19)
    OpacityMask
    Normal
    Tangent
    WorldPositionOffset
    SubsurfaceColor
    ClearCoat
    ClearCoatRoughness
    AmbientOcclusion
    Refraction
    CustomizedUVs_0
    CustomizedUVs_1
    CustomizedUVs_2
    CustomizedUVs_3
    CustomizedUVs_4
    CustomizedUVs_5
    CustomizedUVs_6
    CustomizedUVs_7
    PixelDepthOffset
    ShadingModel
    Displacement
   )) {materialFunction=/Game/Effects/Niagara/MasterMaterials/Functions/MF_FX_MAEmissiveAdjustments}
 ]
}
