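# Additive unlit particle material for lightning strike effects on sprites and Niagara mesh particles.
# A panning normal map warps tiled UVs of a panning lightning sprite texture, with switchable dynamic parameters feeding timing.
# Three sines on per-particle randomized time form a contrast-adjusted flicker, scaled by particle alpha and an optional vertical sphere-mask fade, then multiplied with the texture.
# Particle color times a dynamic multiplier feeds an emissive adjustment function.
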Material /DealershipSharedLibrary/FX/Legacy/M_Lightning_Strike
graph {
 properties {
  bUsedWithNiagaraMeshParticles=true
  bUsedWithNiagaraSprites=true
  bUsedWithParticleSprites=true
  blendMode=BLEND_Additive
  pixelDepthOffsetMode=PDOM_Legacy
  shadingModel=MSM_Unlit
  usedShadingModels=Unlit
 }
 nodes [
  n1 = ParticleColor()
  MaterialFunctionCall(Emissive=Multiply(A=DynamicParameter() {paramNames [TimeOffset, DistortionSpeed, EmissiveMulti, Param4]}.EmissiveMulti, B=@n1.RGB)) {materialFunction=&MF_FX_EmissiveAdjustments}
  n2 = TextureCoordinate()
  n3 = Constant() {r=0.5}
  n4 = SphereMask(A=Power(Base=ComponentMask(None=@n2) {g=true}), B=@n3, Radius=@n3) {hardnessPercent=70}
  n5 = ParticleRandom()
  n6 = Multiply(A=Add(A=Time(), B=Add(A=@n5)), B=Reroute())
  n7 = Sine(None=@n6) {period=0.11}
  n8 = Multiply(A=Sine(None=@n6) {period=0.09}, B=Sine(None=@n6) {period=0.075})
  n9 = MaterialFunctionCall(In=ConstantBiasScale(None=Multiply(A=@n7, B=@n8)), Contrast=ScalarParameter() {defaultValue=3, parameterName=FlickerContrast}) {materialFunction=&CheapContrast}
  n10 = Multiply(A=Max(A=Saturate(None=@n9.Result), B=NamedRerouteUsage()), B=Clamp(None=@n1.A) {clampMode=CMODE_ClampMax})
  n11 = ScalarParameter() {defaultValue=1, parameterName=DistortionTile_X}
  n12 = ScalarParameter() {defaultValue=1, parameterName=DistortionTile_Y}
  n13 = TextureCoordinate()
  n14 = AppendVector(A=ScalarParameter() {defaultValue=1, parameterName=Texture1_Tile_X}, B=ScalarParameter() {defaultValue=1, parameterName=Texture1_Tile_X})
  n15 = DynamicParameter() {paramNames [TimeOffset, DistortionSpeed, EmissiveMulti, Param4]}
  n16 = Constant()
  n17 = StaticSwitchParameter(True=@n15.TimeOffset, False=@n16) {defaultValue=true, parameterName=DynParam_Time_Disable}
  n18 = Multiply(A=Time(), B=@n17)
  n19 = StaticSwitchParameter(True=@n15.DistortionSpeed, False=@n16) {defaultValue=true, parameterName=DynParam_Distotion_Disable}
  n20 = TextureSampleParameter2D(UVs=Panner(
    Coordinate=Multiply(A=@n13, B=@n14)
    Time=Add(A=@n18, B=@n19)
    Speed=VectorParameter() {parameterName=Texture1_Direction, defaultValue=<rgba 1 1 0 1>}.RGB
   )) {parameterName=Texture_1, samplerType=SAMPLERTYPE_Normal, texture=/Game/Effects/Fort_Effects/Creatures/Elemental/Ice/TilingCloudNormalsMap}
  n21 = ComponentMask(None=@n20.RGB) {g=true, r=true}
  n22 = AppendVector(A=ScalarParameter() {defaultValue=0.05, parameterName=Distortion_X}, B=ScalarParameter() {defaultValue=0.1, parameterName=Distortion_Y})
  n23 = ScalarParameter() {defaultValue=1, parameterName=Position_X}
  n24 = ScalarParameter() {defaultValue=1, parameterName=Position_Y}
  n25 = Multiply(A=Multiply(A=TextureCoordinate(), B=AppendVector(A=@n11, B=@n12)), B=Add(A=Multiply(A=@n21, B=@n22), B=AppendVector(A=@n23, B=@n24)))
  n26 = Multiply(A=Time(), B=@n17)
  n27 = VectorParameter() {parameterName=Distortion_Direction, defaultValue=<rgba 0 -1 0 1>}
  Multiply(
   A=StaticSwitchParameter(True=Multiply(A=@n4, B=@n10), False=@n10) {defaultValue=true, parameterName=UseTopBottomFade}
   B=TextureSampleParameter2D(UVs=Panner(Coordinate=@n25, Time=@n26, Speed=@n27.RGB) {speedY=0.5}) {parameterName=Texture, texture=/Game/Effects/Fort_Effects/Textures/Electric/T_Lightning_Sprites}.R
  )
 ]
}
