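# Particle material for sprite-sheet smoke, with optional noise-driven fire, gradient-map coloring and particle color.
# Sub-UV frame math or panned UVs, plus panning distortion textures, drive the smoke and custom opacity samples, with channels selectable through dynamic parameters.
# Opacity can use alpha or a custom mask, with depth fade and dither functions.
# Most features toggle through scalar-driven lerps instead of static switches; Low quality drops secondary distortion.
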Material /DealershipSharedLibrary/FX/Legacy/M_ProtoSmoke_SpriteSheet
graph {
 expressions {
  MaterialExpressionAdd_0 {
   class=MaterialExpressionAdd
   inputs [
    A {from=MaterialExpressionPanner_0}
    B {from=MaterialExpressionComponentMask_1}
   ]
  }
  MaterialExpressionAdd_1 {
   class=MaterialExpressionAdd
   inputs [
    A {from=MaterialExpressionNamedRerouteDeclaration_8}
    B {from=MaterialExpressionScalarParameter_1}
   ]
  }
  MaterialExpressionAdd_2 {
   class=MaterialExpressionAdd
   inputs [
    A {from=MaterialExpressionLinearInterpolate_7}
    B {from=MaterialExpressionLinearInterpolate_8}
   ]
  }
  MaterialExpressionAdd_3 {
   class=MaterialExpressionAdd
   inputs [
    A {from=MaterialExpressionAdd_4}
    B {from=MaterialExpressionMultiply_4}
   ]
  }
  MaterialExpressionAdd_4 {
   class=MaterialExpressionAdd
   inputs [
    A {from=MaterialExpressionNamedRerouteUsage_19}
    B {from=MaterialExpressionScalarParameter_17}
   ]
   properties {constB=-0.04}
  }
  MaterialExpressionAdd_6 {
   class=MaterialExpressionAdd
   inputs [
    A {from=MaterialExpressionAppendVector_1}
    B {from=MaterialExpressionPanner_5}
   ]
  }
  MaterialExpressionAppendVector_1 {
   class=MaterialExpressionAppendVector
   inputs [
    A {from=MaterialExpressionFmod_1, output=""}
    B {from=MaterialExpressionFloor_1, output=""}
   ]
   properties {desc=IndexA_Combined}
  }
  MaterialExpressionComponentMask_0 {class=MaterialExpressionComponentMask, inputs [None {from=MaterialExpressionVectorParameter_3}], properties {g=true, r=true}}
  MaterialExpressionComponentMask_1 {class=MaterialExpressionComponentMask, inputs [None {from=MaterialExpressionNamedRerouteUsage_14}], properties {g=true}}
  MaterialExpressionComponentMask_11 {class=MaterialExpressionComponentMask, inputs [None {from=MaterialExpressionDivide_1}], properties {r=true}}
  MaterialExpressionComponentMask_12 {class=MaterialExpressionComponentMask, inputs [None {from=MaterialExpressionVectorParameter_14, output=RGB}], properties {g=true, r=true}}
  MaterialExpressionComponentMask_2 {class=MaterialExpressionComponentMask, inputs [None {from=MaterialExpressionVectorParameter_6}], properties {g=true, r=true}}
  MaterialExpressionComponentMask_3 {class=MaterialExpressionComponentMask, inputs [None {from=MaterialExpressionVectorParameter_7}], properties {g=true, r=true}}
  MaterialExpressionComponentMask_4 {class=MaterialExpressionComponentMask, inputs [None {from=MaterialExpressionLinearInterpolate_10}], properties {g=true, r=true}}
  MaterialExpressionComponentMask_5 {class=MaterialExpressionComponentMask, inputs [None {from=MaterialExpressionVectorParameter_10, output=RGB}], properties {g=true, r=true}}
  MaterialExpressionComponentMask_7 {class=MaterialExpressionComponentMask, inputs [None {from=MaterialExpressionNamedRerouteUsage_5}], properties {g=true, r=true}}
  MaterialExpressionComponentMask_8 {class=MaterialExpressionComponentMask, inputs [None {from=MaterialExpressionVectorParameter_13, output=RGB}], properties {g=true, r=true}}
  MaterialExpressionConstant3Vector_6 {class=MaterialExpressionConstant3Vector, properties {constant=<rgba 1 0 0 0>}}
  MaterialExpressionConstant3Vector_7 {class=MaterialExpressionConstant3Vector, properties {constant=<rgba 0 1 0 0>}}
  MaterialExpressionConstant3Vector_8 {class=MaterialExpressionConstant3Vector, properties {constant=<rgba 0 0 1 0>}}
  MaterialExpressionConstant_1 {class=MaterialExpressionConstant, properties {r=1}}
  MaterialExpressionConvert_0 {
   class=MaterialExpressionConvert
   inputs [
    None {from=MaterialExpressionVectorParameter_10, output=Speed X}
    None {from=MaterialExpressionVectorParameter_10, output=Speed X}
   ]
   properties {
    nodeName=Make Vector2
    convertInputs [
     {
      type=Scalar
      expressionInput {
       expression=&self:MaterialExpressionVectorParameter_10
       inputName=None
       mask=1
       maskA=0
       maskB=1
       maskG=0
       maskR=0
       outputIndex=3
      }
     }
     {
      type=Scalar
      expressionInput {
       expression=&self:MaterialExpressionVectorParameter_10
       inputName=None
       mask=1
       maskA=1
       maskB=0
       maskG=0
       maskR=0
       outputIndex=4
      }
     }
    ]
    convertOutputs [{type=Vector2}]
   }
  }
  MaterialExpressionConvert_2 {
   class=MaterialExpressionConvert
   inputs [
    None {from=MaterialExpressionVectorParameter_14, output=Speed X}
    None {from=MaterialExpressionVectorParameter_14, output=Speed X}
   ]
   properties {
    nodeName=Make Vector2
    convertInputs [
     {
      type=Scalar
      expressionInput {
       expression=&self:MaterialExpressionVectorParameter_14
       inputName=None
       mask=1
       maskA=0
       maskB=1
       maskG=0
       maskR=0
       outputIndex=3
      }
     }
     {
      type=Scalar
      expressionInput {
       expression=&self:MaterialExpressionVectorParameter_14
       inputName=None
       mask=1
       maskA=1
       maskB=0
       maskG=0
       maskR=0
       outputIndex=4
      }
     }
    ]
    convertOutputs [{type=Vector2}]
   }
  }
  MaterialExpressionDesaturation_0 {
   class=MaterialExpressionDesaturation
   inputs [
    None {from=MaterialExpressionTextureSampleParameter2D_0, output=RGB}
    Fraction {from=MaterialExpressionScalarParameter_22, output=""}
   ]
   properties {}
  }
  MaterialExpressionDivide_1 {
   class=MaterialExpressionDivide
   inputs [
    A {from=MaterialExpressionConstant_1}
    B {from=MaterialExpressionComponentMask_8}
   ]
  }
  MaterialExpressionDotProduct_0 {
   class=MaterialExpressionDotProduct
   inputs [
    A {from=MaterialExpressionTextureSampleParameter2D_1, output=RGBA}
    B {from=MaterialExpressionVectorParameter_0, output=""}
   ]
   properties {}
  }
  MaterialExpressionDotProduct_1 {
   class=MaterialExpressionDotProduct
   inputs [
    A {from=MaterialExpressionNamedRerouteDeclaration_10, output=""}
    B {from=MaterialExpressionNamedRerouteUsage_8, output=""}
   ]
   properties {}
  }
  MaterialExpressionDotProduct_2 {
   class=MaterialExpressionDotProduct
   inputs [
    A {from=MaterialExpressionTextureSampleParameter2D_2, output=RGBA}
    B {from=MaterialExpressionNamedRerouteUsage_9, output=""}
   ]
   properties {}
  }
  MaterialExpressionDotProduct_3 {
   class=MaterialExpressionDotProduct
   inputs [
    A {from=MaterialExpressionTextureSampleParameter2D_3, output=RGBA}
    B {from=MaterialExpressionVectorParameter_4, output=RGB}
   ]
   properties {}
  }
  MaterialExpressionDotProduct_4 {
   class=MaterialExpressionDotProduct
   inputs [
    A {from=MaterialExpressionTextureSampleParameter2D_4, output=RGBA}
    B {from=MaterialExpressionVectorParameter_5, output=RGB}
   ]
   properties {}
  }
  MaterialExpressionDotProduct_5 {
   class=MaterialExpressionDotProduct
   inputs [
    A {from=MaterialExpressionNamedRerouteUsage_1, output=""}
    B {from=MaterialExpressionVectorParameter_8, output=""}
   ]
   properties {}
  }
  MaterialExpressionDynamicParameter_1 {
   class=MaterialExpressionDynamicParameter
   inputs []
   properties {defaultValue=<rgba 0 0 0 1>, paramNames [IsPainted, SmokeCustomSelectorchannel, OpacityCustomSelectorChannel, ManualSubUVFrame]}
  }
  MaterialExpressionFloor_1 {class=MaterialExpressionFloor, inputs [None {from=MaterialExpressionMultiply_6, output=""}], properties {desc=IndexA_Vert}}
  MaterialExpressionFmod_1 {
   class=MaterialExpressionFmod
   inputs [
    A {from=MaterialExpressionSubtract_1, output=""}
    B {from=MaterialExpressionReroute_12, output=""}
   ]
   properties {desc=IndexA_Horz}
  }
  MaterialExpressionFrac_1 {class=MaterialExpressionFrac, inputs [None {from=MaterialExpressionNamedRerouteUsage_17, output=""}], properties {desc=Frame Lerp}}
  MaterialExpressionIf_0 {
   class=MaterialExpressionIf
   inputs [
    A {from=MaterialExpressionScalarParameter_10, output=""}
    B {from="", output=""}
    "A > B" {from=MaterialExpressionStaticComponentMaskParameter_0, output=""}
    "A == B" {from=MaterialExpressionTextureCoordinate_3, output=""}
    "A < B" {from=MaterialExpressionTextureCoordinate_3, output=""}
   ]
   properties {constB=0.99}
  }
  MaterialExpressionIf_1 {
   class=MaterialExpressionIf
   inputs [
    A {from=MaterialExpressionScalarParameter_10, output=""}
    B {from="", output=""}
    "A > B" {from=MaterialExpressionParticleMacroUV_0, output=""}
    "A == B" {from=MaterialExpressionIf_0, output=""}
    "A < B" {from=MaterialExpressionIf_0, output=""}
   ]
   properties {constB=1.99}
  }
  MaterialExpressionIf_2 {
   class=MaterialExpressionIf
   inputs [
    A {from=MaterialExpressionScalarParameter_7, output=""}
    B {from="", output=""}
    "A > B" {from=MaterialExpressionNamedRerouteUsage_20, output=""}
    "A == B" {from=MaterialExpressionPanner_3, output=""}
    "A < B" {from=MaterialExpressionPanner_3, output=""}
   ]
   properties {constB=0.5}
  }
  MaterialExpressionLinearInterpolate_0 {
   class=MaterialExpressionLinearInterpolate
   inputs [
    A {from=MaterialExpressionNamedRerouteUsage_10}
    B {from=MaterialExpressionDotProduct_2}
    Alpha {from=MaterialExpressionMaterialFunctionCall_0, output=Result}
   ]
  }
  MaterialExpressionLinearInterpolate_1 {
   class=MaterialExpressionLinearInterpolate
   inputs [
    A {from=MaterialExpressionLinearInterpolate_15}
    B {from=MaterialExpressionNamedRerouteUsage_3}
    Alpha {from=MaterialExpressionSaturate_3}
   ]
  }
  MaterialExpressionLinearInterpolate_10 {
   class=MaterialExpressionLinearInterpolate
   inputs [
    A {from=MaterialExpressionLinearInterpolate_7}
    B {from=MaterialExpressionQualitySwitch_0}
    Alpha {from=MaterialExpressionSaturate_2}
   ]
  }
  MaterialExpressionLinearInterpolate_11 {
   class=MaterialExpressionLinearInterpolate
   inputs [
    A {from=MaterialExpressionMultiply_2}
    B {from=MaterialExpressionMultiply_19}
    Alpha {from=MaterialExpressionNamedRerouteUsage_2}
   ]
  }
  MaterialExpressionLinearInterpolate_12 {
   class=MaterialExpressionLinearInterpolate
   inputs [
    A {from=MaterialExpressionReroute_5}
    B {from=MaterialExpressionDesaturation_0}
    Alpha {from=MaterialExpressionSaturate_4}
   ]
  }
  MaterialExpressionLinearInterpolate_13 {
   class=MaterialExpressionLinearInterpolate
   inputs [
    A {from=MaterialExpressionLinearInterpolate_5}
    B {from=MaterialExpressionLinearInterpolate_4}
    Alpha {from=MaterialExpressionSaturate_7}
   ]
  }
  MaterialExpressionLinearInterpolate_14 {
   class=MaterialExpressionLinearInterpolate
   inputs [
    A {from=MaterialExpressionReroute_11}
    B {from=MaterialExpressionTextureSample_0, output=RGB}
    Alpha {from=MaterialExpressionSaturate_8}
   ]
  }
  MaterialExpressionLinearInterpolate_15 {
   class=MaterialExpressionLinearInterpolate
   inputs [
    A {from=MaterialExpressionReroute_2}
    B {from=MaterialExpressionMax_0}
    Alpha {from=MaterialExpressionSaturate_9}
   ]
  }
  MaterialExpressionLinearInterpolate_2 {
   class=MaterialExpressionLinearInterpolate
   inputs [
    A {from=MaterialExpressionPower_3}
    B {from=MaterialExpressionMultiply_17}
    Alpha {from=MaterialExpressionSaturate_6}
   ]
  }
  MaterialExpressionLinearInterpolate_3 {
   class=MaterialExpressionLinearInterpolate
   inputs [
    A {from=MaterialExpressionNamedRerouteUsage_0}
    B {from=MaterialExpressionMultiply_18}
    Alpha {from=MaterialExpressionNamedRerouteUsage_13}
   ]
  }
  MaterialExpressionLinearInterpolate_4 {
   class=MaterialExpressionLinearInterpolate
   inputs [
    A {from=MaterialExpressionLinearInterpolate_3}
    B {from=MaterialExpressionLinearInterpolate_5}
    Alpha {from=MaterialExpressionScalarParameter_27}
   ]
  }
  MaterialExpressionLinearInterpolate_5 {
   class=MaterialExpressionLinearInterpolate
   inputs [
    A {from=MaterialExpressionReroute_8}
    B {from=MaterialExpressionLinearInterpolate_11}
    Alpha {from=MaterialExpressionSaturate_5}
   ]
  }
  MaterialExpressionLinearInterpolate_6 {
   class=MaterialExpressionLinearInterpolate
   inputs [
    A {from=MaterialExpressionReroute_6}
    B {from=MaterialExpressionMaterialFunctionCall_2, output=Result}
    Alpha {from=MaterialExpressionScalarParameter_6}
   ]
  }
  MaterialExpressionLinearInterpolate_7 {
   class=MaterialExpressionLinearInterpolate
   inputs [
    A {from=MaterialExpressionDotProduct_3}
    B {from=MaterialExpressionTextureSampleParameter2D_3, output=RGB}
    Alpha {from=MaterialExpressionSaturate_0}
   ]
  }
  MaterialExpressionLinearInterpolate_8 {
   class=MaterialExpressionLinearInterpolate
   inputs [
    A {from=MaterialExpressionDotProduct_4}
    B {from=MaterialExpressionTextureSampleParameter2D_4, output=RGB}
    Alpha {from=MaterialExpressionSaturate_1}
   ]
  }
  MaterialExpressionLinearInterpolate_9 {
   class=MaterialExpressionLinearInterpolate
   inputs [
    A {from=MaterialExpressionAdd_2}
    B {from=MaterialExpressionMultiply_1}
    Alpha {from=MaterialExpressionRound_0}
   ]
  }
  MaterialExpressionMaterialFunctionCall_0 {
   class=MaterialExpressionMaterialFunctionCall
   inputs [Value {from=MaterialExpressionScalarParameter_0}]
   properties {materialFunction=/DealershipSharedLibrary/FX/Materials/MaterialFunctions/MF_DLR_ValueSelectorNormalized}
  }
  MaterialExpressionMaterialFunctionCall_1 {
   class=MaterialExpressionMaterialFunctionCall
   inputs [
    Opacity In {from=MaterialExpressionNamedRerouteUsage_4}
    DepthFade Distance {from=""}
   ]
   properties {materialFunction=&MF_VCVFX_DepthFade}
  }
  MaterialExpressionMaterialFunctionCall_2 {
   class=MaterialExpressionMaterialFunctionCall
   inputs [
    Alpha Threshold {from=MaterialExpressionReroute_6}
    Offset {from=""}
    Random {from=""}
    Offset2 {from=""}
   ]
   properties {materialFunction=&DitherTemporalAA_Fortnite}
  }
  MaterialExpressionMaterialFunctionCall_6 {
   class=MaterialExpressionMaterialFunctionCall
   inputs [Emissive {from=MaterialExpressionLinearInterpolate_14}]
   properties {materialFunction=&MF_FX_EmissiveAdjustments}
  }
  MaterialExpressionMaterialFunctionCall_7 {
   class=MaterialExpressionMaterialFunctionCall
   inputs [
    X {from=MaterialExpressionVectorParameter_6, output=Speed X}
    Y {from=MaterialExpressionVectorParameter_6, output=Speed X}
   ]
   properties {materialFunction=&MakeFloat2}
  }
  MaterialExpressionMaterialFunctionCall_8 {
   class=MaterialExpressionMaterialFunctionCall
   inputs [
    X {from=MaterialExpressionVectorParameter_3, output=PannerSpeedX}
    Y {from=MaterialExpressionVectorParameter_3, output=PannerSpeedX}
   ]
   properties {materialFunction=&MakeFloat2}
  }
  MaterialExpressionMaterialFunctionCall_9 {
   class=MaterialExpressionMaterialFunctionCall
   inputs [
    X {from=MaterialExpressionVectorParameter_7, output=Speed X}
    Y {from=MaterialExpressionVectorParameter_7, output=Speed X}
   ]
   properties {materialFunction=&MakeFloat2}
  }
  MaterialExpressionMax_0 {
   class=MaterialExpressionMax
   inputs [
    A {from=MaterialExpressionReroute_2}
    B {from=MaterialExpressionNamedRerouteDeclaration_12}
   ]
  }
  MaterialExpressionMultiply_0 {
   class=MaterialExpressionMultiply
   inputs [
    A {from=MaterialExpressionLinearInterpolate_0}
    B {from=MaterialExpressionParticleColor_0, output=A}
   ]
  }
  MaterialExpressionMultiply_1 {
   class=MaterialExpressionMultiply
   inputs [
    A {from=MaterialExpressionLinearInterpolate_7}
    B {from=MaterialExpressionLinearInterpolate_8}
   ]
  }
  MaterialExpressionMultiply_10 {
   class=MaterialExpressionMultiply
   inputs [
    A {from=MaterialExpressionTextureCoordinate_1}
    B {from=MaterialExpressionComponentMask_3}
   ]
  }
  MaterialExpressionMultiply_11 {
   class=MaterialExpressionMultiply
   inputs [
    A {from=MaterialExpressionAdd_6}
    B {from=MaterialExpressionReroute_1}
   ]
   properties {desc=UV_A}
  }
  MaterialExpressionMultiply_12 {
   class=MaterialExpressionMultiply
   inputs [
    A {from=MaterialExpressionTextureCoordinate_6}
    B {from=MaterialExpressionComponentMask_12}
   ]
  }
  MaterialExpressionMultiply_16 {
   class=MaterialExpressionMultiply
   inputs [
    A {from=MaterialExpressionMultiply_3}
    B {from=MaterialExpressionScalarParameter_24}
   ]
  }
  MaterialExpressionMultiply_17 {
   class=MaterialExpressionMultiply
   inputs [
    A {from=MaterialExpressionPower_3}
    B {from=MaterialExpressionDotProduct_0}
   ]
  }
  MaterialExpressionMultiply_18 {
   class=MaterialExpressionMultiply
   inputs [
    A {from=MaterialExpressionLinearInterpolate_2}
    B {from=MaterialExpressionNamedRerouteDeclaration_6}
   ]
  }
  MaterialExpressionMultiply_19 {
   class=MaterialExpressionMultiply
   inputs [
    A {from=MaterialExpressionNamedRerouteUsage_0}
    B {from=MaterialExpressionNamedRerouteUsage_12}
   ]
  }
  MaterialExpressionMultiply_2 {
   class=MaterialExpressionMultiply
   inputs [
    A {from=MaterialExpressionNamedRerouteUsage_6}
    B {from=MaterialExpressionVectorParameter_9, output=RGB}
   ]
  }
  MaterialExpressionMultiply_3 {
   class=MaterialExpressionMultiply
   inputs [
    A {from=MaterialExpressionAdd_1}
    B {from=MaterialExpressionScalarParameter_2}
   ]
  }
  MaterialExpressionMultiply_4 {
   class=MaterialExpressionMultiply
   inputs [
    A {from=MaterialExpressionComponentMask_4}
    B {from=MaterialExpressionScalarParameter_13}
   ]
  }
  MaterialExpressionMultiply_5 {
   class=MaterialExpressionMultiply
   inputs [
    A {from=MaterialExpressionTextureCoordinate_5}
    B {from=MaterialExpressionComponentMask_5}
   ]
  }
  MaterialExpressionMultiply_6 {
   class=MaterialExpressionMultiply
   inputs [
    A {from=MaterialExpressionSubtract_1}
    B {from=MaterialExpressionComponentMask_11}
   ]
  }
  MaterialExpressionMultiply_7 {
   class=MaterialExpressionMultiply
   inputs [
    A {from=MaterialExpressionIf_1}
    B {from=MaterialExpressionComponentMask_0}
   ]
  }
  MaterialExpressionMultiply_8 {
   class=MaterialExpressionMultiply
   inputs [
    A {from=MaterialExpressionTextureCoordinate_0}
    B {from=MaterialExpressionComponentMask_2}
   ]
  }
  MaterialExpressionNamedRerouteDeclaration_0 {
   class=MaterialExpressionNamedRerouteDeclaration
   inputs [None {from=MaterialExpressionMultiply_16}]
   properties {name=Smoke Texture RGB, nodeColor=<rgba 1 0.4219 0 1>}
  }
  MaterialExpressionNamedRerouteDeclaration_1 {
   class=MaterialExpressionNamedRerouteDeclaration
   inputs [None {from=MaterialExpressionMultiply_0}]
   properties {name=Opacity Mask, nodeColor=<rgba 0.1016 0 1 1>}
  }
  MaterialExpressionNamedRerouteDeclaration_10 {
   class=MaterialExpressionNamedRerouteDeclaration
   inputs [None {from=MaterialExpressionTextureSampleParameter2D_0, output=RGBA}]
   properties {name=SmokeTextureRGBA, nodeColor=<rgba 0.6406 1 0 1>}
  }
  MaterialExpressionNamedRerouteDeclaration_11 {
   class=MaterialExpressionNamedRerouteDeclaration
   inputs [None {from=MaterialExpressionTextureSampleParameter2D_0, output=A}]
   properties {name=SmokeTextureAlpha, nodeColor=<rgba 0 0.8125 1 1>}
  }
  MaterialExpressionNamedRerouteDeclaration_12 {
   class=MaterialExpressionNamedRerouteDeclaration
   inputs [None {from=MaterialExpressionDotProduct_5}]
   properties {name=SecondaryColorMask, nodeColor=<rgba 1 0.8672 0 1>}
  }
  MaterialExpressionNamedRerouteDeclaration_13 {
   class=MaterialExpressionNamedRerouteDeclaration
   inputs [None {from=MaterialExpressionDynamicParameter_1, output=ManualSubUVFrame}]
   properties {name=Dyn_SubUVFrame, nodeColor=<rgba 0.2891 1 0 1>}
  }
  MaterialExpressionNamedRerouteDeclaration_15 {
   class=MaterialExpressionNamedRerouteDeclaration
   inputs [None {from=MaterialExpressionIf_2}]
   properties {name=Smoke UVs, nodeColor=<rgba 0 1 0.8359 1>}
  }
  MaterialExpressionNamedRerouteDeclaration_16 {
   class=MaterialExpressionNamedRerouteDeclaration
   inputs [None {from=MaterialExpressionMultiply_11}]
   properties {name=CustomOPUVS, nodeColor=<rgba 1 0 0.2344 1>}
  }
  MaterialExpressionNamedRerouteDeclaration_2 {
   class=MaterialExpressionNamedRerouteDeclaration
   inputs [None {from=MaterialExpressionLinearInterpolate_13}]
   properties {name=Textures, nodeColor=<rgba 0.5937 1 0 1>}
  }
  MaterialExpressionNamedRerouteDeclaration_3 {
   class=MaterialExpressionNamedRerouteDeclaration
   inputs [None {from=MaterialExpressionLinearInterpolate_1}]
   properties {name=Smoke Alpha, nodeColor=<rgba 0 0.1797 1 1>}
  }
  MaterialExpressionNamedRerouteDeclaration_4 {
   class=MaterialExpressionNamedRerouteDeclaration
   inputs [None {from=MaterialExpressionSwitch_4}]
   properties {name=Smoke Selector, nodeColor=<rgba 1 0.0567 0.7743 1>}
  }
  MaterialExpressionNamedRerouteDeclaration_5 {
   class=MaterialExpressionNamedRerouteDeclaration
   inputs [None {from=MaterialExpressionSwitch_5}]
   properties {name=CustomOpacityMaskSelector, nodeColor=<rgba 1 0 0.5625 1>}
  }
  MaterialExpressionNamedRerouteDeclaration_6 {
   class=MaterialExpressionNamedRerouteDeclaration
   inputs [None {from=MaterialExpressionParticleColor_1, output=RGB}]
   properties {name=ParticleColorRGB, nodeColor=<rgba 0.0312 0 1 1>}
  }
  MaterialExpressionNamedRerouteDeclaration_7 {
   class=MaterialExpressionNamedRerouteDeclaration
   inputs [None {from=MaterialExpressionDynamicParameter_1, output=OpacityCustomSelectorChannel}]
   properties {name=Dyn_Opacityselector, nodeColor=<rgba 0.0312 0 1 1>}
  }
  MaterialExpressionNamedRerouteDeclaration_8 {
   class=MaterialExpressionNamedRerouteDeclaration
   inputs [None {from=MaterialExpressionLinearInterpolate_12}]
   properties {name=SmokeRGB, nodeColor=<rgba 1 0 0.6094 1>}
  }
  MaterialExpressionNamedRerouteDeclaration_9 {
   class=MaterialExpressionNamedRerouteDeclaration
   inputs [None {from=MaterialExpressionAdd_3}]
   properties {name=Distortion, nodeColor=<rgba 1 0.6563 0 1>}
  }
  MaterialExpressionNamedRerouteUsage_0 {class=MaterialExpressionNamedRerouteUsage}
  MaterialExpressionNamedRerouteUsage_1 {class=MaterialExpressionNamedRerouteUsage}
  MaterialExpressionNamedRerouteUsage_10 {class=MaterialExpressionNamedRerouteUsage}
  MaterialExpressionNamedRerouteUsage_11 {class=MaterialExpressionNamedRerouteUsage}
  MaterialExpressionNamedRerouteUsage_12 {class=MaterialExpressionNamedRerouteUsage}
  MaterialExpressionNamedRerouteUsage_13 {class=MaterialExpressionNamedRerouteUsage}
  MaterialExpressionNamedRerouteUsage_14 {class=MaterialExpressionNamedRerouteUsage}
  MaterialExpressionNamedRerouteUsage_15 {class=MaterialExpressionNamedRerouteUsage}
  MaterialExpressionNamedRerouteUsage_17 {class=MaterialExpressionNamedRerouteUsage}
  MaterialExpressionNamedRerouteUsage_18 {class=MaterialExpressionNamedRerouteUsage}
  MaterialExpressionNamedRerouteUsage_19 {class=MaterialExpressionNamedRerouteUsage}
  MaterialExpressionNamedRerouteUsage_2 {class=MaterialExpressionNamedRerouteUsage}
  MaterialExpressionNamedRerouteUsage_20 {class=MaterialExpressionNamedRerouteUsage}
  MaterialExpressionNamedRerouteUsage_3 {class=MaterialExpressionNamedRerouteUsage}
  MaterialExpressionNamedRerouteUsage_4 {class=MaterialExpressionNamedRerouteUsage}
  MaterialExpressionNamedRerouteUsage_5 {class=MaterialExpressionNamedRerouteUsage}
  MaterialExpressionNamedRerouteUsage_6 {class=MaterialExpressionNamedRerouteUsage}
  MaterialExpressionNamedRerouteUsage_7 {class=MaterialExpressionNamedRerouteUsage}
  MaterialExpressionNamedRerouteUsage_8 {class=MaterialExpressionNamedRerouteUsage}
  MaterialExpressionNamedRerouteUsage_9 {class=MaterialExpressionNamedRerouteUsage}
  MaterialExpressionPanner_0 {
   class=MaterialExpressionPanner
   inputs [
    Coordinate {from=MaterialExpressionMultiply_7}
    Time {from=""}
    Speed {from=MaterialExpressionMaterialFunctionCall_8, output=Result}
   ]
   properties {speedX=0.25, speedY=0.25}
  }
  MaterialExpressionPanner_1 {
   class=MaterialExpressionPanner
   inputs [
    Coordinate {from=MaterialExpressionMultiply_8}
    Time {from=""}
    Speed {from=MaterialExpressionMaterialFunctionCall_7, output=Result}
   ]
  }
  MaterialExpressionPanner_2 {
   class=MaterialExpressionPanner
   inputs [
    Coordinate {from=MaterialExpressionMultiply_10}
    Time {from=""}
    Speed {from=MaterialExpressionMaterialFunctionCall_9, output=Result}
   ]
  }
  MaterialExpressionPanner_3 {
   class=MaterialExpressionPanner
   inputs [
    Coordinate {from=MaterialExpressionMultiply_5}
    Time {from=""}
    Speed {from=MaterialExpressionConvert_0}
   ]
  }
  MaterialExpressionPanner_5 {
   class=MaterialExpressionPanner
   inputs [
    Coordinate {from=MaterialExpressionMultiply_12}
    Time {from=""}
    Speed {from=MaterialExpressionConvert_2}
   ]
  }
  MaterialExpressionParticleColor_0 {class=MaterialExpressionParticleColor, inputs [], properties {}}
  MaterialExpressionParticleColor_1 {class=MaterialExpressionParticleColor, inputs [], properties {}}
  MaterialExpressionParticleMacroUV_0 {class=MaterialExpressionParticleMacroUV, inputs [], properties {}}
  MaterialExpressionPower_3 {
   class=MaterialExpressionPower
   inputs [
    Base {from=MaterialExpressionNamedRerouteUsage_11}
    Exp {from=MaterialExpressionScalarParameter_25}
   ]
  }
  MaterialExpressionQualitySwitch_0 {
   class=MaterialExpressionQualitySwitch
   inputs [
    Default {from=MaterialExpressionLinearInterpolate_7, output=""}
    Low {from=MaterialExpressionLinearInterpolate_7, output=""}
    High {from=MaterialExpressionLinearInterpolate_9, output=""}
    Medium {from=MaterialExpressionLinearInterpolate_9, output=""}
    Epic {from=MaterialExpressionLinearInterpolate_9, output=""}
   ]
   properties {}
  }
  MaterialExpressionReroute_1 {class=MaterialExpressionReroute, inputs [None {from=MaterialExpressionDivide_1}]}
  MaterialExpressionReroute_11 {class=MaterialExpressionReroute, inputs [None {from=MaterialExpressionNamedRerouteUsage_5}]}
  MaterialExpressionReroute_12 {class=MaterialExpressionReroute, inputs [None {from=MaterialExpressionReroute_3}]}
  MaterialExpressionReroute_2 {class=MaterialExpressionReroute, inputs [None {from=MaterialExpressionDotProduct_1}]}
  MaterialExpressionReroute_3 {class=MaterialExpressionReroute, inputs [None {from=MaterialExpressionVectorParameter_13, output=Rows}]}
  MaterialExpressionReroute_5 {class=MaterialExpressionReroute, inputs [None {from=MaterialExpressionReroute_2}]}
  MaterialExpressionReroute_6 {class=MaterialExpressionReroute, inputs [None {from=MaterialExpressionNamedRerouteUsage_4}]}
  MaterialExpressionReroute_7 {class=MaterialExpressionReroute, inputs [None {from=MaterialExpressionLinearInterpolate_6}]}
  MaterialExpressionReroute_8 {class=MaterialExpressionReroute, inputs [None {from=MaterialExpressionMultiply_19}]}
  MaterialExpressionRound_0 {class=MaterialExpressionRound, inputs [None {from=MaterialExpressionScalarParameter_15, output=""}], properties {}}
  MaterialExpressionSaturate_0 {class=MaterialExpressionSaturate, inputs [None {from=MaterialExpressionScalarParameter_11}]}
  MaterialExpressionSaturate_1 {class=MaterialExpressionSaturate, inputs [None {from=MaterialExpressionScalarParameter_12}]}
  MaterialExpressionSaturate_2 {class=MaterialExpressionSaturate, inputs [None {from=MaterialExpressionScalarParameter_18}]}
  MaterialExpressionSaturate_3 {class=MaterialExpressionSaturate, inputs [None {from=MaterialExpressionScalarParameter_4}]}
  MaterialExpressionSaturate_4 {class=MaterialExpressionSaturate, inputs [None {from=MaterialExpressionScalarParameter_23}]}
  MaterialExpressionSaturate_5 {class=MaterialExpressionSaturate, inputs [None {from=MaterialExpressionScalarParameter_3}]}
  MaterialExpressionSaturate_6 {class=MaterialExpressionSaturate, inputs [None {from=MaterialExpressionScalarParameter_26}]}
  MaterialExpressionSaturate_7 {class=MaterialExpressionSaturate, inputs [None {from=MaterialExpressionScalarParameter_28}]}
  MaterialExpressionSaturate_8 {class=MaterialExpressionSaturate, inputs [None {from=MaterialExpressionScalarParameter_29}]}
  MaterialExpressionSaturate_9 {class=MaterialExpressionSaturate, inputs [None {from=MaterialExpressionScalarParameter_5}]}
  MaterialExpressionScalarParameter_0 {
   class=MaterialExpressionScalarParameter
   properties {
    controlType=Enumeration
    enumeration=/DealershipSharedLibrary/FX/Materials/Enums/Enum_DLR_TrueFalse
    group=Opacity
    parameterName="UseCustomOpacity? 0=false 1=True"
    sortPriority=1
   }
  }
  MaterialExpressionScalarParameter_1 {
   class=MaterialExpressionScalarParameter
   properties {desc=Add the value to the selected channel of the smoke texture, group=Main Texture, parameterName=SmokeTextureBias, sortPriority=3}
  }
  MaterialExpressionScalarParameter_10 {
   class=MaterialExpressionScalarParameter
   properties {
    controlType=Enumeration
    enumeration=/DealershipSharedLibrary/FX/Materials/Enums/Enum_DLR_MaterialUVSelector
    group=Noise/Fire
    parameterName=Noise Texture Switch Cordinates
    sortPriority=7
   }
  }
  MaterialExpressionScalarParameter_11 {
   class=MaterialExpressionScalarParameter
   properties {
    controlType=Enumeration
    defaultValue=1
    enumeration=&Enum_DLR_TrueFalse
    group=Distortion
    parameterName=UseRGBChannelDistortion
    sortPriority=4
   }
  }
  MaterialExpressionScalarParameter_12 {
   class=MaterialExpressionScalarParameter
   properties {
    controlType=Enumeration
    defaultValue=1
    enumeration=&Enum_DLR_TrueFalse
    group=Distortion
    parameterName=UseRGBChannelDistortionSecondary
    sortPriority=9
   }
  }
  MaterialExpressionScalarParameter_13 {class=MaterialExpressionScalarParameter, properties {group=Distortion, parameterName=Distortion Strength, sortPriority=0}}
  MaterialExpressionScalarParameter_15 {
   class=MaterialExpressionScalarParameter
   properties {
    controlType=Enumeration
    enumeration=&Enum_DLR_TrueFalse
    group=Distortion
    parameterName="Distortion Blend Mode Add=false Multiply=True"
    sliderMax=1
    sortPriority=6
   }
  }
  MaterialExpressionScalarParameter_17 {class=MaterialExpressionScalarParameter, properties {group=Distortion, parameterName=Offset Distortion, sortPriority=1}}
  MaterialExpressionScalarParameter_18 {
   class=MaterialExpressionScalarParameter
   properties {
    controlType=Enumeration
    enumeration=&Enum_DLR_TrueFalse
    group=Distortion
    parameterName=UseSecondaryDistortion?
    sliderMax=1
    sortPriority=5
   }
  }
  MaterialExpressionScalarParameter_2 {
   class=MaterialExpressionScalarParameter
   properties {defaultValue=1, desc=Multiply the smoke texture bias, group=Main Texture, parameterName=SmokeTextureBiasScale, sortPriority=4}
  }
  MaterialExpressionScalarParameter_22 {class=MaterialExpressionScalarParameter, properties {group=Main Texture, parameterName=Desaturation, sliderMax=1, sortPriority=1}}
  MaterialExpressionScalarParameter_23 {
   class=MaterialExpressionScalarParameter
   properties {
    controlType=Enumeration
    desc=Use channel alpha from smoke texture to mask the result from biasScale
    enumeration=&Enum_DLR_TrueFalse
    group=Main Texture
    parameterName=UseColorFromRGB?
    sortPriority=2
   }
  }
  MaterialExpressionScalarParameter_24 {class=MaterialExpressionScalarParameter, properties {defaultValue=1, group=Main Texture, parameterName=Smoke Darkness, sortPriority=5}}
  MaterialExpressionScalarParameter_25 {class=MaterialExpressionScalarParameter, properties {defaultValue=1, group=Noise/Fire, parameterName=FireAmount, sortPriority=5}}
  MaterialExpressionScalarParameter_26 {
   class=MaterialExpressionScalarParameter
   properties {
    controlType=Enumeration
    enumeration=&Enum_DLR_TrueFalse
    group=Noise/Fire
    parameterName=NoiseDetail?
    sortPriority=2
   }
  }
  MaterialExpressionScalarParameter_27 {class=MaterialExpressionScalarParameter, properties {defaultValue=1, group=Noise/Fire, parameterName=LerpToSmoke, sortPriority=1}}
  MaterialExpressionScalarParameter_28 {
   class=MaterialExpressionScalarParameter
   properties {
    controlType=Enumeration
    enumeration=&Enum_DLR_TrueFalse
    group=Noise/Fire
    parameterName=UseFire?
    sortPriority=0
   }
  }
  MaterialExpressionScalarParameter_29 {
   class=MaterialExpressionScalarParameter
   properties {
    controlType=Enumeration
    enumeration=&Enum_DLR_TrueFalse
    group=GradientMap
    parameterName=UseGradientMap?
    sortPriority=0
   }
  }
  MaterialExpressionScalarParameter_3 {
   class=MaterialExpressionScalarParameter
   properties {
    controlType=Enumeration
    enumeration=&Enum_DLR_TrueFalse
    group=Main Texture
    parameterName=UseSecondaryMaskColor
    sortPriority=6
   }
  }
  MaterialExpressionScalarParameter_4 {
   class=MaterialExpressionScalarParameter
   properties {
    controlType=Enumeration
    desc=Use channel alpha from smoke texture to mask the result from biasScale
    enumeration=&Enum_DLR_TrueFalse
    group=Opacity
    parameterName=UseAlphaChannelFromSmokeTexture?
    sortPriority=0
   }
  }
  MaterialExpressionScalarParameter_5 {
   class=MaterialExpressionScalarParameter
   properties {
    controlType=Enumeration
    enumeration=&Enum_DLR_TrueFalse
    group=Main Texture
    parameterName=UseSecondaryMaskColor
    sortPriority=6
   }
  }
  MaterialExpressionScalarParameter_6 {
   class=MaterialExpressionScalarParameter
   properties {
    controlType=Enumeration
    enumeration=&Enum_DLR_TrueFalse
    group=Opacity
    parameterName=UseDitherOpacity?
   }
  }
  MaterialExpressionScalarParameter_7 {class=MaterialExpressionScalarParameter, properties {parameterName=UseSubUV UVs for Opacity And Color?}}
  MaterialExpressionStaticComponentMaskParameter_0 {
   class=MaterialExpressionStaticComponentMaskParameter
   inputs [None {from=MaterialExpressionWorldPosition_0, output=XYZ}]
   properties {defaultG=true, defaultR=true, group=Noise/Fire, parameterName=Noise Texture WorldPosition Mask, sortPriority=8}
  }
  MaterialExpressionSubtract_1 {
   class=MaterialExpressionSubtract
   inputs [
    A {from=MaterialExpressionNamedRerouteUsage_17}
    B {from=MaterialExpressionFrac_1}
   ]
   properties {desc=IndexA}
  }
  MaterialExpressionSwitch_4 {
   class=MaterialExpressionSwitch
   inputs [
    SwitchValue {from=MaterialExpressionDynamicParameter_1, output=SmokeCustomSelectorchannel}
    Default {from=MaterialExpressionVectorParameter_1, output=""}
    None {from=MaterialExpressionVectorParameter_1, output=""}
    R {from=MaterialExpressionConstant3Vector_6, output=""}
    G {from=MaterialExpressionConstant3Vector_7, output=""}
    B {from=MaterialExpressionConstant3Vector_8, output=""}
   ]
   properties {inputs [
     {
      inputName=None
      input {
       expression=&self:MaterialExpressionVectorParameter_1
       inputName=None
       mask=1
       maskA=0
       maskB=1
       maskG=1
       maskR=1
       outputIndex=0
      }
     }
     {
      inputName=R
      input {
       expression=&self:MaterialExpressionConstant3Vector_6
       inputName=None
       mask=1
       maskA=0
       maskB=1
       maskG=1
       maskR=1
       outputIndex=0
      }
     }
     {
      inputName=G
      input {
       expression=&self:MaterialExpressionConstant3Vector_7
       inputName=None
       mask=1
       maskA=0
       maskB=1
       maskG=1
       maskR=1
       outputIndex=0
      }
     }
     {
      inputName=B
      input {
       expression=&self:MaterialExpressionConstant3Vector_8
       inputName=None
       mask=1
       maskA=0
       maskB=1
       maskG=1
       maskR=1
       outputIndex=0
      }
     }
    ]}
  }
  MaterialExpressionSwitch_5 {
   class=MaterialExpressionSwitch
   inputs [
    SwitchValue {from=MaterialExpressionNamedRerouteUsage_7, output=""}
    Default {from=MaterialExpressionVectorParameter_2, output=""}
    None {from=MaterialExpressionVectorParameter_2, output=""}
    R {from=MaterialExpressionConstant3Vector_6, output=""}
    G {from=MaterialExpressionConstant3Vector_7, output=""}
    B {from=MaterialExpressionConstant3Vector_8, output=""}
   ]
   properties {inputs [
     {
      inputName=None
      input {
       expression=&self:MaterialExpressionVectorParameter_2
       inputName=None
       mask=1
       maskA=0
       maskB=1
       maskG=1
       maskR=1
       outputIndex=0
      }
     }
     {
      inputName=R
      input {
       expression=&self:MaterialExpressionConstant3Vector_6
       inputName=None
       mask=1
       maskA=0
       maskB=1
       maskG=1
       maskR=1
       outputIndex=0
      }
     }
     {
      inputName=G
      input {
       expression=&self:MaterialExpressionConstant3Vector_7
       inputName=None
       mask=1
       maskA=0
       maskB=1
       maskG=1
       maskR=1
       outputIndex=0
      }
     }
     {
      inputName=B
      input {
       expression=&self:MaterialExpressionConstant3Vector_8
       inputName=None
       mask=1
       maskA=0
       maskB=1
       maskG=1
       maskR=1
       outputIndex=0
      }
     }
    ]}
  }
  MaterialExpressionTextureCoordinate_0 {class=MaterialExpressionTextureCoordinate}
  MaterialExpressionTextureCoordinate_1 {class=MaterialExpressionTextureCoordinate}
  MaterialExpressionTextureCoordinate_2 {class=MaterialExpressionTextureCoordinate}
  MaterialExpressionTextureCoordinate_3 {class=MaterialExpressionTextureCoordinate}
  MaterialExpressionTextureCoordinate_5 {class=MaterialExpressionTextureCoordinate}
  MaterialExpressionTextureCoordinate_6 {class=MaterialExpressionTextureCoordinate}
  MaterialExpressionTextureObjectParameter_0 {
   class=MaterialExpressionTextureObjectParameter
   inputs [
    UVs {from="", output=""}
    Apply View MipBias {from="", output=""}
   ]
   properties {group=GradientMap, parameterName=GradientMap, sortPriority=1}
  }
  MaterialExpressionTextureSampleParameter2D_0 {
   class=MaterialExpressionTextureSampleParameter2D
   inputs [
    UVs {from=MaterialExpressionNamedRerouteUsage_15, output=""}
    Level {from="", output=""}
    Apply View MipBias {from="", output=""}
   ]
   properties {
    group=Main Texture
    mipValueMode=TMVM_MipLevel
    parameterName=SmokeTexture
    sortPriority=0
    texture=/Game/Effects/Fort_Effects/Textures/Smoke/T_Smoke_Thick_8x8_02
   }
  }
  MaterialExpressionTextureSampleParameter2D_1 {
   class=MaterialExpressionTextureSampleParameter2D
   inputs [
    UVs {from=MaterialExpressionAdd_0, output=""}
    Apply View MipBias {from="", output=""}
   ]
   properties {
    group=Noise/Fire
    parameterName=NoiseTexture
    samplerType=SAMPLERTYPE_LinearGrayscale
    sortPriority=3
    texture=/Game/Effects/Fort_Effects/Emotes/UltraLight/Texture/T_Tiling_Noise_01_GreyScale
   }
  }
  MaterialExpressionTextureSampleParameter2D_2 {
   class=MaterialExpressionTextureSampleParameter2D
   inputs [
    UVs {from=MaterialExpressionNamedRerouteUsage_18, output=""}
    Apply View MipBias {from="", output=""}
   ]
   properties {group=Opacity, parameterName=CustomOpacityTexture, texture=&T_Smoke_Thick_8x8_02}
  }
  MaterialExpressionTextureSampleParameter2D_3 {
   class=MaterialExpressionTextureSampleParameter2D
   inputs [
    UVs {from=MaterialExpressionPanner_1, output=""}
    Apply View MipBias {from="", output=""}
   ]
   properties {group=Distortion, parameterName=Distortion Texture, sortPriority=2, texture=/DealershipSharedLibrary/FX/Textures/Noises/T_DLR_Flow03_D}
  }
  MaterialExpressionTextureSampleParameter2D_4 {
   class=MaterialExpressionTextureSampleParameter2D
   inputs [
    UVs {from=MaterialExpressionPanner_2, output=""}
    Apply View MipBias {from="", output=""}
   ]
   properties {
    group=Distortion
    parameterName=Secondary Distortion Texture
    sortPriority=7
    texture=&T_DLR_Flow03_D
   }
  }
  MaterialExpressionTextureSample_0 {
   class=MaterialExpressionTextureSample
   inputs [
    UVs {from=MaterialExpressionComponentMask_7}
    Tex {from=MaterialExpressionTextureObjectParameter_0}
    Apply View MipBias {from=""}
   ]
  }
  MaterialExpressionVectorParameter_0 {
   class=MaterialExpressionVectorParameter
   properties {
    group=Noise/Fire
    parameterName=Noise TEX RGBA Mask
    sortPriority=4
    channelNames {a=ALPHA, b=BLUE, g=GREEN, r=RED}
    defaultValue=<rgba 0 1 0 0>
   }
  }
  MaterialExpressionVectorParameter_1 {
   class=MaterialExpressionVectorParameter
   properties {
    group=Main Texture
    parameterName=Smoke TEX RGBA Mask
    sortPriority=1
    channelNames {a=ALPHA, b=BLUE, g=GREEN, r=RED}
    defaultValue=<rgba 0 1 0 0>
   }
  }
  MaterialExpressionVectorParameter_10 {
   class=MaterialExpressionVectorParameter
   properties {group=Main Texture, parameterName=SmokeTextureUVs, channelNames {a=Speed Y, b=Speed X, g=Tiling Y, r=Tiling X}, defaultValue=<rgba 1 1 0 0>}
  }
  MaterialExpressionVectorParameter_13 {
   class=MaterialExpressionVectorParameter
   properties {parameterName=CustomOpacity SubUV, channelNames {a="", b="", g=Columns, r=Rows}, defaultValue=<rgba 1 1 0 0>}
  }
  MaterialExpressionVectorParameter_14 {
   class=MaterialExpressionVectorParameter
   properties {group=Main Texture, parameterName=CustomOpacityUVs, channelNames {a=Speed Y, b=Speed X, g=Tiling Y, r=Tiling X}, defaultValue=<rgba 1 1 0 0>}
  }
  MaterialExpressionVectorParameter_2 {
   class=MaterialExpressionVectorParameter
   properties {
    group=Opacity
    parameterName=CustomOpacity TEX RGBA Mask
    sortPriority=2
    channelNames {a=ALPHA, b=BLUE, g=GREEN, r=RED}
    defaultValue=<rgba 1 0 0 0>
   }
  }
  MaterialExpressionVectorParameter_3 {
   class=MaterialExpressionVectorParameter
   properties {
    group=Noise/Fire
    parameterName=NoiseTextureUVs
    sortPriority=6
    channelNames {a=PannerSpeedY, b=PannerSpeedX, g=Tiling Y, r=Tiling X}
    defaultValue=<rgba 1 1 0.25 0.25>
   }
  }
  MaterialExpressionVectorParameter_4 {
   class=MaterialExpressionVectorParameter
   properties {group=Distortion, parameterName=Distortion Selector RGBA Mask, sortPriority=4, defaultValue=<rgba 1 0 0 0>}
  }
  MaterialExpressionVectorParameter_5 {
   class=MaterialExpressionVectorParameter
   properties {group=Distortion, parameterName=Secondary Distortion Selector RGBA Mask, sortPriority=8, defaultValue=<rgba 1 0 0 0>}
  }
  MaterialExpressionVectorParameter_6 {
   class=MaterialExpressionVectorParameter
   properties {
    group=Distortion
    parameterName=Distortion Uvs
    sortPriority=3
    channelNames {a=Speed Y, b=Speed X, g=Tiling Y, r=Tiling X}
    defaultValue=<rgba 1 1 0.05 0.2>
   }
  }
  MaterialExpressionVectorParameter_7 {
   class=MaterialExpressionVectorParameter
   properties {
    group=Distortion
    parameterName=Secondary Distortion Uvs
    sortPriority=7
    channelNames {a=Speed Y, b=Speed X, g=Tiling Y, r=Tiling X}
    defaultValue=<rgba 1 1 0.1 -0.3>
   }
  }
  MaterialExpressionVectorParameter_8 {
   class=MaterialExpressionVectorParameter
   properties {
    group=Main Texture
    parameterName=Smoke Secondary Color RGBA Mask
    sortPriority=7
    channelNames {a=ALPHA, b=BLUE, g=GREEN, r=RED}
    defaultValue=<rgba 0 1 0 0>
   }
  }
  MaterialExpressionVectorParameter_9 {
   class=MaterialExpressionVectorParameter
   properties {group=Main Texture, parameterName=CustomSecondaryColor, sortPriority=8, defaultValue=<rgba 1 1 1 0>}
  }
  MaterialExpressionWorldPosition_0 {class=MaterialExpressionWorldPosition, inputs [], properties {}}
 }
 properties {
  bUsedWithNiagaraMeshParticles=true
  bUsedWithNiagaraSprites=true
  blendMode=BLEND_Translucent
  opacityMaskClipValue=0.333
  shadingModel=MSM_Unlit
  usedShadingModels=Unlit
 }
}
