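# Unlit additive smoke material for Niagara sprites, ribbons and meshes.
# Two panning flow textures, reduced to one on low and medium quality, distort a panning main noise texture.
# The contrast- and brightness-adjusted noise multiplies a channel-selectable mask and particle color, then feeds an emissive adjustment function.
# A separate mask and particle alpha product gets depth fade, which the mobile shading path skips.
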Material /DealershipSharedLibrary/FX/Legacy/M_VCVFX_Smoke_DragonCat
graph {
 properties {
  bAutomaticallySetUsageInEditor=false
  bUsedWithNiagaraMeshParticles=true
  bUsedWithNiagaraRibbons=true
  bUsedWithNiagaraSprites=true
  blendMode=BLEND_Additive
  shadingModel=MSM_Unlit
  usedShadingModels=Unlit
 }
 nodes [
  n1 = ScalarParameter() {defaultValue=1, group=General, parameterName=Time Panner Distortion OG, sortPriority=1}
  n2 = ScalarParameter() {group=General, parameterName=Speed X Distortion OG, sortPriority=2}
  n3 = ScalarParameter() {group=General, parameterName=Speed Y Distortion OG, sortPriority=3}
  n4 = TextureSampleParameter2D(UVs=MaterialFunctionCall!(
    ScaleUVByCenter X=ScalarParameter() {defaultValue=1, group=Distortion Texture 1, parameterName=ScaleUVByCenter X Distort Tex 1}
    ScaleUVByCenter Y=ScalarParameter() {defaultValue=1, group=Distortion Texture 1, parameterName=ScaleUVByCenter Y Distort Tex 1}
    Tiling X=ScalarParameter() {defaultValue=1, group=Distortion Texture 1, parameterName=TilingX Distort Texture 1}
    Tiling Y=ScalarParameter() {defaultValue=1, group=Distortion Texture 1, parameterName=TilingY Distort Texture 1}
    Panner Time=ScalarParameter() {defaultValue=1, group=Distortion Texture 1, parameterName=Panner Time Distort Texture 1}
    Panner Speed X=ScalarParameter() {group=Distortion Texture 1, parameterName=Panner Speed X Distort Texture 1}
    Panner Speed Y=ScalarParameter() {group=Distortion Texture 1, parameterName=Panner Speed Y Distort Texture 1}
    Scale U
    Scale V
   ) {materialFunction=/DealershipSharedLibrary/FX/Materials/MaterialFunctions/MF_VCVFX_TilingAndOffset}.Result) {group=Distortion Texture 1, parameterName=Distortion Tex 1, texture=/DealershipSharedLibrary/FX/Textures/Noise/T_VCVFX_Flow03_N}
  n5 = TextureSampleParameter2D(UVs=MaterialFunctionCall(
    ScaleUVByCenter X=ScalarParameter() {defaultValue=1, group=Distortion Texture 2, parameterName=ScaleUVByCenter X Distort Tex 2}
    ScaleUVByCenter Y=ScalarParameter() {defaultValue=1, group=Distortion Texture 2, parameterName=ScaleUVByCenter Y Distort Tex 2}
    Tiling X=ScalarParameter() {defaultValue=1, group=Distortion Texture 2, parameterName=TilingX Distort Texture 2}
    Tiling Y=ScalarParameter() {defaultValue=1, group=Distortion Texture 2, parameterName=TilingY Distort Texture 2}
    Panner Time=ScalarParameter() {defaultValue=1, group=Distortion Texture 2, parameterName=Panner Time Distort Texture 2}
    Panner Speed X=ScalarParameter() {group=Distortion Texture 2, parameterName=Panner Speed X Distort Texture 2}
    Panner Speed Y=ScalarParameter() {group=Distortion Texture 2, parameterName=Panner Speed Y Distort Texture 2}
   ) {materialFunction=&MF_VCVFX_TilingAndOffset}.Result) {group=Distortion Texture 2, parameterName=Distortion Tex 2, texture=&T_VCVFX_Flow03_N}
  n6 = Add(A=@n4.RGB, B=@n5.RGB)
  n7 = QualitySwitch(Default=@n6, Low=@n4.RGB, High=@n6, Medium=@n4.RGB, Epic=@n6) {desc=Disable second texture on low end devices}
  n8 = Add(
   A=Panner(
    Coordinate=TextureCoordinate()
    Time=MaterialFunctionCall(Speed=@n1) {materialFunction=&TimeWithSpeedVariable}.Time
    Speed=MaterialFunctionCall(X=@n2, Y=@n3) {materialFunction=&MakeFloat2}.Result
   )
   B=Multiply(A=ComponentMask(None=@n7) {g=true, r=true}, B=ScalarParameter() {defaultValue=0.3, group=General, parameterName=Distortion Strength})
  )
  n9 = Multiply(
   A=Power(
    Base=TextureSampleParameter2D(UVs=@n8) {group=General, parameterName=MainTexture, texture=/DealershipSharedLibrary/FX/Textures/Noise/T_VCVFX_Water01_A}.RGB
    Exp=ScalarParameter() {defaultValue=1, group=General, parameterName="Contrast (POW)"}
   )
   B=ScalarParameter() {defaultValue=1, group=General, parameterName="Brightness (Mul)"}
  )
  n10 = TextureSampleParameter2D(UVs=MaterialFunctionCall(
    ScaleUVByCenter X=ScalarParameter() {defaultValue=1, group=Mask, parameterName=ScaleUVByCenter X}
    ScaleUVByCenter Y=ScalarParameter() {defaultValue=1, group=Mask, parameterName=ScaleUVByCenter Y}
    Tiling X=ScalarParameter() {defaultValue=1, group=Mask, parameterName=TilingX Mask}
    Tiling Y=ScalarParameter() {defaultValue=1, group=Mask, parameterName=TilingY Mask}
    Panner Time=ScalarParameter() {group=Mask, parameterName=Panner Time Mask}
    Panner Speed X=ScalarParameter() {group=Mask, parameterName=Panner Speed X Mask}
    Panner Speed Y=ScalarParameter() {group=Mask, parameterName=Panner Speed Y MAsk}
   ) {materialFunction=&MF_VCVFX_TilingAndOffset}.Result) {group=Mask, parameterName=Texture Mask, texture=/DealershipSharedLibrary/FX/Textures/Gradient_Masks/T_VCVFX_FlameCloud_02}
  n11 = Multiply(
   A=Power(
    Base=StaticComponentMaskParameter(None=@n10.RGB) {defaultB=true, defaultG=true, defaultR=true, group=Mask, parameterName=Mask Channel}
    Exp=ScalarParameter() {defaultValue=1, group=Mask, parameterName="Contrast Mask (POW)"}
   )
   B=ScalarParameter() {defaultValue=1, group=Mask, parameterName="Brightness Mask (Mul)"}
  )
  n12 = ParticleColor()
  MaterialFunctionCall(Emissive=Multiply(A=Multiply(A=@n9, B=@n11), B=@n12.RGB)) {materialFunction=&MF_FX_EmissiveAdjustments}
  n13 = Multiply(A=@n11, B=@n12.A)
  n14 = Multiply(
   A=@n13
   B=DepthFade(FadeDistance=ScalarParameter() {defaultValue=100, group=Opacity, parameterName=DepthFade-Standard_Distance, sortPriority=0}) {fadeDistanceDefault=20}
  )
  ShadingPathSwitch(Default=@n14, Deferred=@n14, Mobile=@n13) {desc=Migrated from FeatureLevelSwitch}
 ]
}
